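# Physics asset for the RenegadeWhip backpack cosmetic, defining bodies and joints for its dangling cat's secondary motion.
# Body and head joint separately to the base with tight limits; tail hangs off the body, ears off the head.
# Star-ring-star and two strap chains join the body through zero-size anchor capsules with free rotation, then use limited joints.
# Shapes are capsules except the boxed star; ears and straps swing mainly on one axis.
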PhysicsAsset /BRCosmetics/Accessories/FORT_Backpacks/Backpack_F_MED_RenegadeWhip/Meshes/F_MED_RenegadeWhip_Pack_Physics
bodies {
 dyn_cat_base [{shapeType=Capsule, length=0, radius=0, shapeName=dyn_cat_base_capsule}]
 dyn_cat_body [{shapeType=Capsule, length=3.1076, radius=6.9015, shapeName=dyn_cat_body_capsule, center=<7.7742 -0.2919 0>}]
 dyn_cat_ear_l_base [{shapeType=Capsule, length=2, radius=3, shapeName=dyn_cat_ear_l_base_capsule, center=<2.5289 0 0>}]
 dyn_cat_ear_r_base [{shapeType=Capsule, length=2, radius=3, shapeName=dyn_cat_ear_r_base_capsule, center=<2.5289 0 0>}]
 dyn_cat_head [{shapeType=Capsule, length=4.1395, radius=9.8369, shapeName=dyn_cat_head_capsule, center=<8.7808 -1.0422 0>}]
 dyn_cat_star [{
   shapeType=Box
   extentX=7.0981
   extentY=0.8488
   extentZ=6.9998
   shapeName=dyn_cat_star_box
   center=<3.0519 -0.038 0.08>
   rotation=<rot 0 89.2954 -88.3163>
  }]
 dyn_cat_star_base [{shapeType=Capsule, length=0, radius=0, shapeName=dyn_cat_star_base_capsule}]
 dyn_cat_star_ring [{
   shapeType=Capsule
   length=2.137
   radius=0.8951
   shapeName=dyn_cat_star_ring_capsule
   center=<1.3754 0.0011 -0.0346>
   rotation=<rot 88.559 0.2083 0.1615>
  }]
 dyn_cat_strap_l_01 [{shapeType=Capsule, length=0.5308, radius=0.5572, shapeName=dyn_cat_strap_l_01_capsule, center=<0.8617 0 0>, rotation=<rot 90 0 0>}]
 dyn_cat_strap_l_base [{shapeType=Capsule, length=0, radius=0, shapeName=dyn_cat_strap_l_base_capsule}]
 dyn_cat_strap_r_01 [{shapeType=Capsule, length=0.5308, radius=0.5572, shapeName=dyn_cat_strap_r_01_capsule, center=<0.8617 0 0>, rotation=<rot 90 0 0>}]
 dyn_cat_strap_r_base [{shapeType=Capsule, length=0, radius=0, shapeName=dyn_cat_strap_r_base_capsule}]
 dyn_cat_tail [{shapeType=Capsule, length=5.4993, radius=3.1593, shapeName=dyn_cat_tail_capsule, center=<4.7352 0 0>, rotation=<rot -90 -180 180>}]
}
constraints [
 {bone1Name=dyn_cat_body, bone2Name=dyn_cat_base, swing1LimitDegrees=10, swing2LimitDegrees=10, twistLimitDegrees=5}
 {
  bone1Name=dyn_cat_star_base
  bone2Name=dyn_cat_body
  swing1LimitDegrees=45
  swing1Motion=Free
  swing2LimitDegrees=45
  swing2Motion=Free
  twistLimitDegrees=45
  twistMotion=Free
 }
 {bone1Name=dyn_cat_star_ring, bone2Name=dyn_cat_star_base, swing1LimitDegrees=40, swing2LimitDegrees=40, twistLimitDegrees=5}
 {
  bone1Name=dyn_cat_strap_l_base
  bone2Name=dyn_cat_body
  swing1LimitDegrees=45
  swing1Motion=Free
  swing2LimitDegrees=45
  swing2Motion=Free
  twistLimitDegrees=45
  twistMotion=Free
 }
 {bone1Name=dyn_cat_strap_l_01, bone2Name=dyn_cat_strap_l_base, swing1LimitDegrees=10, swing2LimitDegrees=1, twistLimitDegrees=0}
 {
  bone1Name=dyn_cat_strap_r_base
  bone2Name=dyn_cat_body
  swing1LimitDegrees=45
  swing1Motion=Free
  swing2LimitDegrees=45
  swing2Motion=Free
  twistLimitDegrees=45
  twistMotion=Free
 }
 {bone1Name=dyn_cat_strap_r_01, bone2Name=dyn_cat_strap_r_base, swing1LimitDegrees=10, swing2LimitDegrees=1, twistLimitDegrees=0}
 {bone1Name=dyn_cat_tail, bone2Name=dyn_cat_body, swing1LimitDegrees=40, swing2LimitDegrees=50, twistLimitDegrees=5}
 {bone1Name=dyn_cat_head, bone2Name=dyn_cat_base, swing1LimitDegrees=3, swing2LimitDegrees=3, twistLimitDegrees=0}
 {bone1Name=dyn_cat_ear_l_base, bone2Name=dyn_cat_head, swing1LimitDegrees=65, swing2LimitDegrees=2, twistLimitDegrees=0}
 {bone1Name=dyn_cat_ear_r_base, bone2Name=dyn_cat_head, swing1LimitDegrees=50, swing2LimitDegrees=2, twistLimitDegrees=0}
 {bone1Name=dyn_cat_star, bone2Name=dyn_cat_star_ring, swing1LimitDegrees=5, swing2LimitDegrees=50, twistLimitDegrees=2}
]
